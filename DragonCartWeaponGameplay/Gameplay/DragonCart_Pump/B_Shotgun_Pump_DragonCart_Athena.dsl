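; Blueprint for the DragonCart pump shotgun weapon in Athena, named for the DragonCart gameplay feature.
; It declares a Niagara particle-data receive event with no logic, and a GetActorAudioParameters override that returns zero.
; The construction script only calls its parent, so all behavior comes from the parent class.
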(event Niagara|EventReceiveParticleData
  (params (Data :type Array%20of%20Basic%20Particle%20Data%20Structures) (NiagaraSystem :type Niagara%20System%20Object%20Reference) (SimulationPositionOffset :type Vector%20%28by%20ref%29 :default "0, 0, 0")))

(fn GetActorAudioParameters
  (return 0))

(fn ConstructionScript
  (|CallParentFunction))
